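# Behavior tree for the Labrador AI, which idles as a "Fake Llama" and topples, jumps or flees through gameplay abilities.
# A priority Selector on BB_Labrador waits while tractor-beamed or launched, then handles forced toppling.
# When Awake-tagged OR GoalActor passes its check, it tries Jump (if MustJump), then Flee; otherwise it idles on a long Wait.
# Abilities are gated by has/can-activate decorators, with short Wait fallbacks.
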BehaviorTree /Labrador/BehaviorTree/BT_Labrador
blackboard=&/BB_Labrador
root {
 name=Selector
 class=BTComposite_Selector
 children [
  Wait {
   class=BTTask_Wait
   decorators [IsCapturedByTractorBeam {
     class=BTDecorator_GameplayTagQuery
     properties {
      flowAbortMode=Both
      nodeName=IsCapturedByTractorBeam
      gameplayTagQuery {autoDescription=" ANY( Gameplay.Status.TractorBeam )", queryTokenStream [0,1,1,1,0], tagDictionary [{tagName=Gameplay.Status.TractorBeam}]}
     }
    }]
   properties {waitTime {defaultValue=2}}
  }
  Selector {
   class=BTComposite_Selector
   children [Launched Wait {class=BTTask_Wait, properties {nodeName=Launched Wait, waitTime {defaultValue=5}}}]
   decorators [Is Launched? {
     class=BTDecorator_Blackboard
     properties {arithmeticOperation=Equal, intValue=0, nodeName=Is Launched?, stringValue="", blackboardKey {selectedKeyName=IsLaunched}}
    }]
   properties {children [{childComposite=null, childTask=&self:BTTask_Wait_3, decorators []}], services []}
   services []
  }
  Selector {
   class=BTComposite_Selector
   children [
    GameplayAbility Topple {
     class=FortBTTask_ExecuteGameplayAbility
     decorators [
      Has Gameplay Ability {
       class=FortBTDecorator_GameplayAbility_HasGameplayAbility
       properties {gameplayAbilityTagContainer {defaultValue {gameplayTags [{tagName=Ability.Labrador.Topple}]}}}
      }
      Can Activate Gameplay Ability {
       class=FortBTDecorator_GameplayAbility_CanActivate
       properties {
        flowAbortMode=LowerPriority
        gameplayAbilityTagContainer {defaultValue {gameplayTags [{tagName=Ability.Labrador.Topple}]}}
        targetValue {key=GoalActor}
       }
      }
     ]
     properties {nodeName=GameplayAbility Topple, gameplayAbilityTagContainer {defaultValue {gameplayTags [{tagName=Ability.Labrador.Topple}]}}}
    }
    Fallback {class=BTTask_Wait, properties {nodeName=Fallback, waitTime {defaultValue=1}}}
   ]
   decorators [Must be Topple? {
     class=BTDecorator_Blackboard
     properties {arithmeticOperation=Equal, intValue=0, nodeName=Must be Topple?, stringValue="", blackboardKey {selectedKeyName=MustBeToppled}}
    }]
   properties {
    children [
     {
      childComposite=null
      childTask=&self:FortBTTask_ExecuteGameplayAbility_0
      decorators [
       &self:FortBTDecorator_GameplayAbility_HasGameplayAbility_1
       &self:FortBTDecorator_GameplayAbility_CanActivate_5
      ]
     }
     {childComposite=null, childTask=&self:BTTask_Wait_0, decorators []}
    ]
    services []
   }
   services []
  }
  Selector {
   class=BTComposite_Selector
   children [
    GameplayAbility Jump {
     class=FortBTTask_ExecuteGameplayAbility
     decorators [
      Must Jump? {
       class=BTDecorator_Blackboard
       properties {
        arithmeticOperation=Equal
        flowAbortMode=LowerPriority
        intValue=0
        nodeName=Must Jump?
        stringValue=""
        blackboardKey {selectedKeyName=MustJump}
       }
      }
      Has Gameplay Ability {
       class=FortBTDecorator_GameplayAbility_HasGameplayAbility
       properties {gameplayAbilityTagContainer {defaultValue {gameplayTags [{tagName=Ability.Labrador.Jump}]}}}
      }
      Can Activate Gameplay Ability {
       class=FortBTDecorator_GameplayAbility_CanActivate
       properties {flowAbortMode=None, gameplayAbilityTagContainer {defaultValue {gameplayTags [{tagName=Ability.Labrador.Jump}]}}, targetValue {key=GoalActor}}
      }
     ]
     properties {nodeName=GameplayAbility Jump, gameplayAbilityTagContainer {defaultValue {gameplayTags [{tagName=Ability.Labrador.Jump}]}}}
    }
    GameplayAbility Flee {
     class=FortBTTask_ExecuteGameplayAbility
     decorators [
      Has Gameplay Ability {
       class=FortBTDecorator_GameplayAbility_HasGameplayAbility
       properties {gameplayAbilityTagContainer {defaultValue {gameplayTags [{tagName=Ability.Labrador.Flee}]}}}
      }
      Can Activate Gameplay Ability {
       class=FortBTDecorator_GameplayAbility_CanActivate
       properties {flowAbortMode=None, gameplayAbilityTagContainer {defaultValue {gameplayTags [{tagName=Ability.Labrador.Flee}]}}, targetValue {key=GoalActor}}
      }
     ]
     properties {nodeName=GameplayAbility Flee, gameplayAbilityTagContainer {defaultValue {gameplayTags [{tagName=Ability.Labrador.Flee}]}}}
    }
    Fallback {class=BTTask_Wait, properties {nodeName=Fallback, waitTime {defaultValue=1}}}
   ]
   decorators [
    Gameplay Tag Query {
     class=BTDecorator_GameplayTagQuery
     properties {
      flowAbortMode=Both
      nodeName=Gameplay Tag Query
      gameplayTagQuery {autoDescription=" ANY( Ability.Labrador.Awake )", queryTokenStream [0,1,1,1,0], tagDictionary [{tagName=Ability.Labrador.Awake}]}
     }
    }
    Blackboard Based Condition {
     class=BTDecorator_Blackboard
     properties {
      arithmeticOperation=Equal
      intValue=0
      nodeName=Blackboard Based Condition
      notifyObserver=ValueChange
      stringValue=""
      blackboardKey {selectedKeyName=GoalActor}
     }
    }
   ]
   properties {
    children [
     {
      childComposite=null
      childTask=&self:FortBTTask_ExecuteGameplayAbility_5
      decorators [
       &self:BTDecorator_Blackboard_2
       &self:FortBTDecorator_GameplayAbility_HasGameplayAbility_0
       &self:FortBTDecorator_GameplayAbility_CanActivate_0
      ]
     }
     {
      childComposite=null
      childTask=&self:FortBTTask_ExecuteGameplayAbility_4
      decorators [
       &self:FortBTDecorator_GameplayAbility_HasGameplayAbility_4
       &self:FortBTDecorator_GameplayAbility_CanActivate_4
      ]
     }
     {childComposite=null, childTask=&self:BTTask_Wait_2, decorators []}
    ]
    services []
   }
   services []
  }
  Idle / Fake Llama {class=BTTask_Wait, properties {nodeName=Idle / Fake Llama, waitTime {defaultValue=3600}}}
 ]
 decorators []
 properties {
  children [
   {
    childComposite=null
    childTask=&self:BTTask_Wait_4
    decorators [&self:BTDecorator_GameplayTagQuery_0]
   }
   {
    childComposite=&self:BTComposite_Selector_2
    childTask=null
    decorators [&self:BTDecorator_Blackboard_0]
   }
   {
    childComposite=&self:BTComposite_Selector_1
    childTask=null
    decorators [&self:BTDecorator_Blackboard_1]
   }
   {
    childComposite=&self:BTComposite_Selector_4
    childTask=null
    decoratorOps [{number=2, operation=Or}, {number=0}, {number=1}]
    decorators [&self:BTDecorator_GameplayTagQuery_1, &self:BTDecorator_Blackboard_4]
   }
   {childComposite=null, childTask=&self:BTTask_Wait_1, decorators []}
  ]
  services []
 }
 services []
}
