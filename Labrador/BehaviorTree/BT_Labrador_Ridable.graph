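# Behavior tree for the ridable Labrador AI, branching on the Riding.IsBeingRidden tag with self-aborting tag queries.
# While ridden, it sets MustJump and waits.
# Unridden, it prioritizes tractor beam, suspension, launch and topple states, then repositions around a set Leader.
# Without a Leader, it jumps or flees through gameplay abilities when awake or given a GoalActor, falling back to a long fake-llama idle.
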BehaviorTree /Labrador/BehaviorTree/BT_Labrador_Ridable
blackboard=&/BB_Labrador
root {
 name=Selector
 class=BTComposite_Selector
 children [
  Selector {
   class=BTComposite_Selector
   children [Sequence {
     class=BTComposite_Sequence
     children [
      SetRidingKey {
       class=Task_SetBlackboardBoolValues_C
       children []
       decorators []
       properties {
        bIgnoreRestartSelf=false
        bShowPropertyDetails=true
        boolKey1Value=true
        boolKey2Value=false
        boolKey3Value=false
        boolKey4Value=false
        boolKey5Value=false
        customDescription=""
        nodeName=SetRidingKey
        update 1?=true
        update 2?=false
        update 3?=false
        update 4?=false
        update 5?=false
        boolKey1 {bNoneIsAllowedValue=false, selectedKeyName=MustJump}
        boolKey2 {bNoneIsAllowedValue=false, selectedKeyName=GoalActor}
        boolKey3 {bNoneIsAllowedValue=false, selectedKeyName=GoalActor}
        boolKey4 {bNoneIsAllowedValue=false, selectedKeyName=GoalActor}
        boolKey5 {bNoneIsAllowedValue=false, selectedKeyName=GoalActor}
        tickInterval {interval=-1}
       }
       services []
      }
      Wait {class=BTTask_Wait, properties {waitTime {defaultValue=2}}}
     ]
     decorators []
     properties {children [
       {childTask=&self:Task_SetBlackboardBoolValues_C_1}
       {childTask=&self:BTTask_Wait_18}
      ]}
    }]
   decorators [isRidden {
     class=BTDecorator_GameplayTagQuery
     properties {
      flowAbortMode=Self
      nodeName=isRidden
      gameplayTagQuery {autoDescription=" ANY( Riding.IsBeingRidden )", queryTokenStream [0,1,1,1,0], tagDictionary [{tagName=Riding.IsBeingRidden}]}
     }
    }]
   properties {children [{childComposite=&self:BTComposite_Sequence_1, childTask=null, decorators []}], services []}
   services []
  }
  Selector {
   class=BTComposite_Selector
   children [
    Selector {
     class=BTComposite_Selector
     children [
      Wait {
       class=BTTask_Wait
       decorators [CapturedByTractorBeamOrSuspended {
         class=BTDecorator_GameplayTagQuery
         properties {
          flowAbortMode=Both
          nodeName=CapturedByTractorBeamOrSuspended
          gameplayTagQuery {
           autoDescription=" ANY( Gameplay.Status.TractorBeam, Gameplay.Status.SuspendedInAir )"
           queryTokenStream [0,1,1,2,0,1]
           tagDictionary [{tagName=Gameplay.Status.TractorBeam}, {tagName=Gameplay.Status.SuspendedInAir}]
          }
         }
        }]
       properties {waitTime {defaultValue=2}}
      }
      Selector {
       class=BTComposite_Selector
       children [Launched Wait {class=BTTask_Wait, properties {nodeName=Launched Wait, waitTime {defaultValue=5}}}]
       decorators [Is Launched? {
         class=BTDecorator_Blackboard
         properties {arithmeticOperation=Equal, intValue=0, nodeName=Is Launched?, stringValue="", blackboardKey {selectedKeyName=IsLaunched}}
        }]
       properties {children [{childComposite=null, childTask=&self:BTTask_Wait_3, decorators []}], services []}
       services []
      }
      Selector {
       class=BTComposite_Selector
       children [
        GameplayAbility Topple {
         class=FortBTTask_ExecuteGameplayAbility
         decorators [
          Has Gameplay Ability {
           class=FortBTDecorator_GameplayAbility_HasGameplayAbility
           properties {gameplayAbilityTagContainer {defaultValue {gameplayTags [{tagName=Ability.Labrador.Topple}]}}}
          }
          Can Activate Gameplay Ability {
           class=FortBTDecorator_GameplayAbility_CanActivate
           properties {
            flowAbortMode=LowerPriority
            gameplayAbilityTagContainer {defaultValue {gameplayTags [{tagName=Ability.Labrador.Topple}]}}
            targetValue {key=GoalActor}
           }
          }
         ]
         properties {nodeName=GameplayAbility Topple, gameplayAbilityTagContainer {defaultValue {gameplayTags [{tagName=Ability.Labrador.Topple}]}}}
        }
        Fallback {class=BTTask_Wait, properties {nodeName=Fallback, waitTime {defaultValue=1}}}
       ]
       decorators [Must be Topple? {
         class=BTDecorator_Blackboard
         properties {arithmeticOperation=Equal, intValue=0, nodeName=Must be Topple?, stringValue="", blackboardKey {selectedKeyName=MustBeToppled}}
        }]
       properties {
        children [
         {
          childComposite=null
          childTask=&self:FortBTTask_ExecuteGameplayAbility_0
          decorators [
           &self:FortBTDecorator_GameplayAbility_HasGameplayAbility_7
           &self:FortBTDecorator_GameplayAbility_CanActivate_6
          ]
         }
         {childComposite=null, childTask=&self:BTTask_Wait_0, decorators []}
        ]
        services []
       }
       services []
      }
     ]
     decorators []
     properties {
      children [
       {
        childComposite=null
        childTask=&self:BTTask_Wait_5
        decorators [&self:BTDecorator_GameplayTagQuery_7]
       }
       {
        childComposite=&self:BTComposite_Selector_2
        childTask=null
        decorators [&self:BTDecorator_Blackboard_9]
       }
       {
        childComposite=&self:BTComposite_Selector_1
        childTask=null
        decorators [&self:BTDecorator_Blackboard_8]
       }
      ]
      services []
     }
     services []
    }
    Selector {
     class=BTComposite_Selector
     children [
      GameplayAbility {
       class=FortBTTask_ExecuteGameplayAbility
       decorators [
        Has Gameplay Ability {
         class=FortBTDecorator_GameplayAbility_HasGameplayAbility
         properties {gameplayAbilityTagContainer {defaultValue {gameplayTags [{tagName=NPC.Ability.NonAttack.Reposition.AroundLeader}]}}}
        }
        Can Activate Gameplay Ability {
         class=FortBTDecorator_GameplayAbility_CanActivate
         properties {
          flowAbortMode=None
          gameplayAbilityTagContainer {defaultValue {gameplayTags [{tagName=NPC.Ability.NonAttack.Reposition.AroundLeader}]}}
          targetValue {key=SpawnSpot}
         }
        }
       ]
       properties {
        nodeName=GameplayAbility
        gameplayAbilityTagContainer {defaultValue {gameplayTags [{tagName=NPC.Ability.NonAttack.Reposition.AroundLeader}]}}
       }
      }
      Wait {
       class=BTTask_Wait
       decorators [Is On Cooldown {
         class=FortBTDecorator_GameplayAbility_IsOnCooldown
         properties {
          bInverseCondition=false
          flowAbortMode=None
          nodeName=Is On Cooldown
          activeAbilityTagsToSkipTestingValue {key=None, defaultValue {gameplayTags []}}
          gameplayAbilityTagContainer {key=None, defaultValue {gameplayTags [{tagName=NPC.Ability.NonAttack.Reposition.AroundLeader}]}}
          targetValue {baseClass=&Actor, defaultValue=null, key=SpawnSpot}
         }
        }]
       properties {waitTime {defaultValue=1}}
      }
     ]
     decorators [Blackboard Based Condition {
       class=BTDecorator_Blackboard
       properties {
        arithmeticOperation=Equal
        intValue=0
        nodeName=Blackboard Based Condition
        notifyObserver=ValueChange
        stringValue=""
        blackboardKey {selectedKeyName=Leader}
       }
      }]
     properties {
      children [
       {
        childComposite=null
        childTask=&self:FortBTTask_ExecuteGameplayAbility_6
        decorators [
         &self:FortBTDecorator_GameplayAbility_HasGameplayAbility_1
         &self:FortBTDecorator_GameplayAbility_CanActivate_0
        ]
       }
       {
        childComposite=null
        childTask=&self:BTTask_Wait_7
        decorators [&self:FortBTDecorator_GameplayAbility_IsOnCooldown_0]
       }
      ]
      services []
     }
     services []
    }
    Selector {
     class=BTComposite_Selector
     children [Selector {
       class=BTComposite_Selector
       children [
        GameplayAbility Jump {
         class=FortBTTask_ExecuteGameplayAbility
         decorators [
          Must Jump? {
           class=BTDecorator_Blackboard
           properties {
            arithmeticOperation=Equal
            flowAbortMode=LowerPriority
            intValue=0
            nodeName=Must Jump?
            stringValue=""
            blackboardKey {selectedKeyName=MustJump}
           }
          }
          Has Gameplay Ability {
           class=FortBTDecorator_GameplayAbility_HasGameplayAbility
           properties {gameplayAbilityTagContainer {defaultValue {gameplayTags [{tagName=Ability.Labrador.Jump}]}}}
          }
          Can Activate Gameplay Ability {
           class=FortBTDecorator_GameplayAbility_CanActivate
           properties {flowAbortMode=None, gameplayAbilityTagContainer {defaultValue {gameplayTags [{tagName=Ability.Labrador.Jump}]}}, targetValue {key=GoalActor}}
          }
         ]
         properties {nodeName=GameplayAbility Jump, gameplayAbilityTagContainer {defaultValue {gameplayTags [{tagName=Ability.Labrador.Jump}]}}}
        }
        GameplayAbility Flee {
         class=FortBTTask_ExecuteGameplayAbility
         decorators [
          Has Gameplay Ability {
           class=FortBTDecorator_GameplayAbility_HasGameplayAbility
           properties {gameplayAbilityTagContainer {defaultValue {gameplayTags [{tagName=Ability.Labrador.Flee}]}}}
          }
          Can Activate Gameplay Ability {
           class=FortBTDecorator_GameplayAbility_CanActivate
           properties {flowAbortMode=None, gameplayAbilityTagContainer {defaultValue {gameplayTags [{tagName=Ability.Labrador.Flee}]}}, targetValue {key=GoalActor}}
          }
         ]
         properties {nodeName=GameplayAbility Flee, gameplayAbilityTagContainer {defaultValue {gameplayTags [{tagName=Ability.Labrador.Flee}]}}}
        }
        Fallback {class=BTTask_Wait, properties {nodeName=Fallback, waitTime {defaultValue=1}}}
       ]
       decorators [
        Gameplay Tag Query {
         class=BTDecorator_GameplayTagQuery
         properties {
          flowAbortMode=Both
          nodeName=Gameplay Tag Query
          gameplayTagQuery {autoDescription=" ANY( Ability.Labrador.Awake )", queryTokenStream [0,1,1,1,0], tagDictionary [{tagName=Ability.Labrador.Awake}]}
         }
        }
        Blackboard Based Condition {
         class=BTDecorator_Blackboard
         properties {
          arithmeticOperation=Equal
          intValue=0
          nodeName=Blackboard Based Condition
          notifyObserver=ValueChange
          stringValue=""
          blackboardKey {selectedKeyName=GoalActor}
         }
        }
       ]
       properties {
        children [
         {
          childComposite=null
          childTask=&self:FortBTTask_ExecuteGameplayAbility_1
          decorators [
           &self:BTDecorator_Blackboard_7
           &self:FortBTDecorator_GameplayAbility_HasGameplayAbility_6
           &self:FortBTDecorator_GameplayAbility_CanActivate_3
          ]
         }
         {
          childComposite=null
          childTask=&self:FortBTTask_ExecuteGameplayAbility_4
          decorators [
           &self:FortBTDecorator_GameplayAbility_HasGameplayAbility_4
           &self:FortBTDecorator_GameplayAbility_CanActivate_4
          ]
         }
         {childComposite=null, childTask=&self:BTTask_Wait_2, decorators []}
        ]
        services []
       }
       services []
      }]
     decorators [Blackboard Based Condition {
       class=BTDecorator_Blackboard
       properties {
        arithmeticOperation=Equal
        basicOperation=NotSet
        intValue=0
        nodeName=Blackboard Based Condition
        notifyObserver=ValueChange
        stringValue=""
        blackboardKey {selectedKeyName=Leader}
       }
      }]
     properties {
      children [{
        childComposite=&self:BTComposite_Selector_4
        childTask=null
        decoratorOps [{number=2, operation=Or}, {number=0}, {number=1}]
        decorators [
         &self:BTDecorator_GameplayTagQuery_1
         &self:BTDecorator_Blackboard_11
        ]
       }]
      services []
     }
     services []
    }
    Idle / Fake Llama {class=BTTask_Wait, properties {nodeName=Idle / Fake Llama, waitTime {defaultValue=3600}}}
   ]
   decorators [notRidden {
     class=BTDecorator_GameplayTagQuery
     properties {
      flowAbortMode=Self
      nodeName=notRidden
      gameplayTagQuery {autoDescription=" NONE( Riding.IsBeingRidden )", queryTokenStream [0,1,3,1,0], tagDictionary [{tagName=Riding.IsBeingRidden}]}
     }
    }]
   properties {
    children [
     {childComposite=&self:BTComposite_Selector_18, childTask=null, decorators []}
     {
      childComposite=&self:BTComposite_Selector_9
      childTask=null
      decorators [&self:BTDecorator_Blackboard_12]
     }
     {
      childComposite=&self:BTComposite_Selector_11
      childTask=null
      decorators [&self:BTDecorator_Blackboard_2]
     }
     {childComposite=null, childTask=&self:BTTask_Wait_1, decorators []}
    ]
    services []
   }
   services []
  }
 ]
 decorators []
 properties {
  children [
   {
    childComposite=&self:BTComposite_Selector_5
    childTask=null
    decorators [&self:BTDecorator_GameplayTagQuery_3]
   }
   {
    childComposite=&self:BTComposite_Selector_6
    childTask=null
    decorators [&self:BTDecorator_GameplayTagQuery_2]
   }
  ]
  services []
 }
 services []
}
